# Copyright Epic Games, Inc. All Rights Reserved.
#################################################
# Generated Digest of Verse API
# DO NOT modify this manually!
# Generated from build: ++Fortnite+Release-36.20-CL-44118662
#################################################


using {/Verse.org/Native}
using {/Verse.org/SceneGraph}
# Module import path: /playkigai.dig@fortnite.com/Minigame_FlipTheSwitch/EPs
EPs := module:
    @import_as("/Minigame_FlipTheSwitch/EPs/damageable_collider.damageable_collider_C")
    damageable_collider<scoped {/playkigai.dig@fortnite.com/Minigame_FlipTheSwitch}> := class<final>(entity):


    damageable_collider_asset<scoped {/playkigai.dig@fortnite.com/Minigame_FlipTheSwitch}>:entity_prefab = external {}

    @import_as("/Minigame_FlipTheSwitch/EPs/flip_switch.flip_switch_C")
    flip_switch<scoped {/playkigai.dig@fortnite.com/Minigame_FlipTheSwitch}> := class<final>(entity):


    flip_switch_asset<scoped {/playkigai.dig@fortnite.com/Minigame_FlipTheSwitch}>:entity_prefab = external {}

    @import_as("/Minigame_FlipTheSwitch/EPs/flip_switch_platform.flip_switch_platform_C")
    flip_switch_platform<scoped {/playkigai.dig@fortnite.com/Minigame_FlipTheSwitch}> := class<final>(entity):


    flip_switch_platform_asset<scoped {/playkigai.dig@fortnite.com/Minigame_FlipTheSwitch}>:entity_prefab = external {}

    @import_as("/Minigame_FlipTheSwitch/EPs/fts_game_template.fts_game_template_C")
    fts_game_template<scoped {/playkigai.dig@fortnite.com/Minigame_FlipTheSwitch}> := class<final>(entity):


    fts_game_template_asset<scoped {/playkigai.dig@fortnite.com/Minigame_FlipTheSwitch}>:entity_prefab = external {}

    @import_as("/Minigame_FlipTheSwitch/EPs/platform_volume.platform_volume_C")
    platform_volume<scoped {/playkigai.dig@fortnite.com/Minigame_FlipTheSwitch}> := class<final>(entity):


    platform_volume_asset<scoped {/playkigai.dig@fortnite.com/Minigame_FlipTheSwitch}>:entity_prefab = external {}

    @import_as("/Minigame_FlipTheSwitch/EPs/raised_platform.raised_platform_C")
    raised_platform<scoped {/playkigai.dig@fortnite.com/Minigame_FlipTheSwitch}> := class<final>(entity):


    raised_platform_asset<scoped {/playkigai.dig@fortnite.com/Minigame_FlipTheSwitch}>:entity_prefab = external {}