<#
@UUID: 9b27e5ac-12db-41b7-8bdc-4a9411782d43
@Task: Damageable Collider Component
@Description: A component that uses a objective_device placed in the level to detect when a scene graph eneity is damaged by a fortnite damage source.
@Tags: SceneGraph, Gameplay, Movement, PlayerInteraction
@Export: true
#>

using { /Verse.org }
using { /Verse.org/Concurrency }
using { /Verse.org/Simulation }
using { /Verse.org/Native }
using { /Verse.org/SceneGraph }
using { /Verse.org/SpatialMath }
using { /Fortnite.com/Game }
using { /Fortnite.com/Playspaces }
using { /Fortnite.com/Devices }
using { /UnrealEngine.com/Temporary/SpatialMath }
using { /UnrealEngine.com/Temporary/Diagnostics }
using {EPs}

<#
In-Editor Entity Prefab Setup:
    (NOTE: any setting not mentioned is assumed to be default or to your preference)

damageable_collider
    RootEntity:
        - damageable_collider_component
            - MaybeObjectiveDevice: Either set to an objective_device placed in the level, or be managed by a parent manager component like flip_switch_platform_component
            - Objective Device Settings:
                - Mesh: Cylinder, Sphere, or Cube
                - Owning Team: None
                - Health: 320000 (as high as you can possibly make it)
                - Collision During Games: "Only When Visible"
                - Collides With :"Weapons Only"

#>


damageable_collider_component<public>:= class<final_super>(component):

    @editable
    var MaybeObjectiveDevice: ?objective_device = false
    @editable
    var UseDeviceScale: logic = true

    var MaybePostTickHandler: ?cancelable = false
    var IsEnabled: logic = true
    

    ColliderHitEvent<public>: event(damage_result) = event(damage_result){}
    EndSimulationEvent<public>: event() = event(){}


    OnBeginSimulation<override>():void=
        (super:)OnBeginSimulation()
        if (RoundMan := Entity.GetFortRoundManager[]):
            RoundMan.SubscribeRoundStarted(OnRoundStarted)


    OnEndSimulation<override>():void=
        (super:)OnEndSimulation()
        if. MaybePostTickHandler?.Cancel()
        EndSimulationEvent.Signal()
        

    OnRoundStarted()<suspends>:void =
        if(ObjectiveDevice := MaybeObjectiveDevice?):
            # Subscribe to post-physics tick events
            set MaybePostTickHandler = option. TickEvents.PostPhysics.Subscribe(OnPostPhysics)

            # At the time of writing, the objective_device has a bug that prevents the damage event from firing until another event is called on it.
            # So we get around it by calling an empty heal function to "wake it up"
            ObjectiveDevice.Heal(0.0)
            ObjectiveDevice.DamagedEvent().Subscribe(OnHit)


    OnPostPhysics(Delta: float):void = 
        if(ObjectiveDevice := MaybeObjectiveDevice?):
            DeviceTransform := ObjectiveDevice.GetTransform()
            GlobalTransform := Entity.GetGlobalTransform()
            NewScale : (/UnrealEngine.com/Temporary/SpatialMath:) vector3 = if (UseDeviceScale?) then DeviceTransform.Scale else FromVector3(GlobalTransform.Scale)

            FinalGlobal :=  (/UnrealEngine.com/Temporary/SpatialMath:) transform{
                Translation := FromVector3(GlobalTransform.Translation),
                Rotation := FromRotation(GlobalTransform.Rotation),
                Scale := NewScale
            }
            
            if. ObjectiveDevice.TeleportTo[FinalGlobal]
            

    OnHit(Result: damage_result):void = 
        if(ObjectiveDevice := MaybeObjectiveDevice?):
            ObjectiveDevice.SetMaxHealth(ObjectiveDevice.GetMaxHealth())
            ColliderHitEvent.Signal(Result)


    Enable<public>():void =
        set IsEnabled = true
        if(ObjectiveDevice := MaybeObjectiveDevice?):
            ObjectiveDevice.Show()
            ObjectiveDevice.SetInvulnerable(false)


    Disable<public>():void =
        set IsEnabled = false
        if(ObjectiveDevice := MaybeObjectiveDevice?):
            ObjectiveDevice.Hide()
            ObjectiveDevice.SetInvulnerable(true)

    
(ColliderPF: damageable_collider).GetColliderComp()<decides><transacts>:damageable_collider_component =
    ColliderPF.FindDescendantComponentsArray(damageable_collider_component)[0]