<#
@UUID: 9b27e5ac-12db-41b7-8bdc-4a9411782d43
@Task: Flip Switch Component
@Description: A scene graph component for a game object that can  be in either a "Flipped" or "Unflipped" state.
    When unflipped, it can be damaged by a damagable_collider_component to cause it to flip. Can unflip automatically
    after a brief cooldown.
@Tags: SceneGraph, Gameplay, Physics, Movement, PlayerInteraction
@Export: true
#>

<#
In-Editor Entity Prefab Setup:
    (NOTE: any setting not mentioned is assumed to be default or to your preference)


damageable_collider
    RootEntity:
        - damageable_collider_component
            - MaybeObjectiveDevice: Either set to an objective_device placed in the level, or be managed by a parent manager component like flip_switch_platform_component
            - Objective Device Settings:
                Mesh: Cylinder, Sphere, or Cube
                Owning Team: None
                Health: 320000 (as high as you can possibly make it)
                Collision During Games: "Only When Visible"
                Collides With :"Weapons Only"


flip_switch
    RootEntity:
        - flip_switch_component
        DamageableCollider (damageable_collider entity prefab instance):
            - damageable_collider_component
                - MaybeObjectiveDevice: should be automatically set by the flip_switch_platform_component
            - transform_component
                - Translation: place slightly above FlipPlate
        FlipPlate:
            - mesh_component (preferrably a cylinder)
            - transform_component
                - Scale: tweak it until it is a flattened circle
                
#>

using { /Verse.org }
using { /Verse.org/Concurrency }
using { /Verse.org/Random }
using { /Verse.org/Simulation }
using { /Verse.org/Native }
using { /Verse.org/SceneGraph }
using { /Verse.org/SpatialMath }
using { /Fortnite.com/Game }
using { /Fortnite.com/Playspaces }
using { /Fortnite.com/Devices }
using { /UnrealEngine.com/Temporary/SpatialMath }
using { /UnrealEngine.com/Temporary/Diagnostics }
using {EPs}



flip_switch_component<public>:= class<final_super>(component):
    # Flag that indicates if a switch should 
    @editable
    var UnflipOnTimer: logic = true

    # Flag that indicates whether this flip switch should start the round in the flipped. Will not automatically unflip even if "UnflipOnTimer" is set until manually unflipped
    @editable
    var StartGameFlipped: logic = false

    # The amount of time in seconds this switch waits after being hit to automatically unflip itself
    @editable_number(float):
        MinValue := option. 0.0
    MinHitCooldown: float = 3.0

    # The max random variation added to switch cooldowns
    @editable_number(float):
        MinValue := option. 0.0
    HitCooldownVariation: float = 3.0

    # The speed in degrees per second that the switch rotates when it flips
    @editable_number(float):
        MinValue := option. 0.0
    FlipSpeed: float = 10.0

    # Entity which has the damageable_collider_component that handles the objective device collider
    @editable
    var MaybeColliderPF: ?damageable_collider = false

    # Entity which has mesh_component that will flip when hit
    @editable
    var MaybeFlipPlateEntity: ?entity = false

    # Events
    # Signaled when the child damageable_collider_component detects a hit
    HitEvent<public>: event(damage_result) = event(damage_result){}
    
    # Signaled when the Flip function is first called regarless if it succeeds or not; used to cancel cooldown that would cause the switch to eventually unflip
    FlipAttemptedEvent<public>: event() = event(){}
    # Signaled at the start of a successful Flip function call, but before the flip animation plays; used to cancel an in-progress Unflip animation into a Flip function call
    FlipStartEvent<public>: event() = event(){}
    # Signaled after the flip animation completes; used when logic needs to execute after the switch finishes moving
    FlipCompleteEvent<public>: event() = event(){}

    # Signaled when the UnFlip function is first called regarless if it succeeds or not; used to cancel cooldown that would cause the switch to eventually flip
    UnflipAttemptedEvent<public>: event() = event(){}
    # Signaled at the start of a successful Unflip function call, but before the unflip animation plays; used to cancel an in-progress Flip animation into a Flip functino call
    UnflipStartEvent<public>: event() = event(){}
    # Signaled after the unflip animation completes; used when logic needs to execute after the switch finishes moving
    UnflipCompleteEvent<public>: event() = event(){}
    # Signaled when this component's OnEndSimulation is called; Used to cancel most asyncronous behavior.
    EndSimulationEvent<public>: event() = event(){}


    var CooldownTimeRemaining: float = 0.0
    var ResetRotation: (/Verse.org/SpatialMath:)rotation = (/Verse.org/SpatialMath:)IdentityRotation()
    # Flag that tracks whether this switch is in the flipped or unflipped state
    var IsFlipped: logic = false
    # Number of degrees that that the local 
    FlipOffset: float = 180.0


    OnBeginSimulation<override>():void=
        (super:)OnBeginSimulation()
        # Check if  Damagable Collider Component is set
        if (ColliderPF := MaybeColliderPF?) {}
        else Print("flip_switch_component\nERROR: couldn't find damageable_collider_component")

        # Try to Get Flip Plate Mesh Component
        if (FlipPlateEntity := MaybeFlipPlateEntity?):
            set ResetRotation = FlipPlateEntity.GetLocalTransform().Rotation
        else:
            Print("flip_switch_component\nERROR: couldn't find flip plate mesh component")
        
        # Set up On Begin Play callbacks
        if (RoundMan := Entity.GetFortRoundManager[]):
            RoundMan.SubscribeRoundStarted(OnRoundStarted)
            

    OnRoundStarted()<suspends>: void =
        # Determine if switch begins game flipped or not
        if (StartGameFlipped?):
            Flip(?PlayAnimation:= false)
        else:
            Unflip(?PlayAnimation:= false)
            
        if:
            ColliderPF := MaybeColliderPF?
            ColliderComp := ColliderPF.GetColliderComp[]
        then:
            race:
                EndSimulationEvent.Await()
                loop:   # On Switch Collider Hit
                    Result := ColliderComp.ColliderHitEvent.Await()
                    if (not IsFlipped?):
                        HitEvent.Signal(Result)
                        Flip()
                        if (UnflipOnTimer?):
                            UnflipCooldown()
                    Sleep(0.0)
                    

    OnEndSimulation<override>():void=
        (super:)OnEndSimulation()
        EndSimulationEvent.Signal()
    
            
    Flip(?PlayAnimation:logic = true)<suspends>:void = 
        FlipAttemptedEvent.Signal()
        if (IsFlipped?) then return
        if (FlipPlateEntity := MaybeFlipPlateEntity?):
    
            # Once this is set to true, Unflip can be called on this component to asyncronously cancel the Flip Animation
            set IsFlipped = true
            FlipStartEvent.Signal()

            race:
                EndSimulationEvent.Await()
                UnflipStartEvent.Await()  # Cancel flip animation if Unflip is called during it
                block:
                    # Disable Collider Comp before beginning animation
                    if (ColliderComp :=  MaybeColliderPF?.GetColliderComp[]):
                        ColliderComp.Disable()
                    
                    TargetRot := FromRotation(FromRotation(ResetRotation).ApplyRoll((/Verse.org/SpatialMath:)DegreesToRadians(FlipOffset)))
                    
                    if (PlayAnimation?):
                        RotateFlipPlateToTargetRotation(TargetRot,FlipSpeed)
                    else: 
                        FinalLocal := (/Verse.org/SpatialMath:) transform{
                            Translation := FlipPlateEntity.GetLocalTransform().Translation,
                            Rotation := TargetRot,
                            Scale := FlipPlateEntity.GetLocalTransform().Scale
                        }
                        FlipPlateEntity.SetLocalTransform(FinalLocal)

                    FlipCompleteEvent.Signal()


    UnflipCooldown(?CooldownTime: float = MinHitCooldown)<suspends>: void = 
        set CooldownTimeRemaining = CooldownTime + GetRandomFloat(0.0, HitCooldownVariation)
        race:
            EndSimulationEvent.Await()
            FlipAttemptedEvent.Await()
            block:
                set CooldownTimeRemaining = MinHitCooldown + GetRandomFloat(0.0, HitCooldownVariation)
                loop:
                    Timestamp := GetSimulationElapsedTime()
                    Sleep(0.0)
                    DeltaTime := GetSimulationElapsedTime() - Timestamp
                    set CooldownTimeRemaining -= DeltaTime
                    if (CooldownTimeRemaining <= 0.0):
                        Unflip()
                        break
        set CooldownTimeRemaining = 0.0
                                

    Unflip(?PlayAnimation:logic = true)<suspends>:void =
        UnflipAttemptedEvent.Signal()
        if (not IsFlipped?) then return

        if (FlipPlateEntity := MaybeFlipPlateEntity?):
            # Once this is set to false, Flip can be called on this component to cancel the Unflip
            set IsFlipped = false
            UnflipStartEvent.Signal()
            race:
                EndSimulationEvent.Await()
                FlipStartEvent.Await()      # Cancel Unflip animation if Flip is called during it
                block:
                    if (PlayAnimation?):
                        RotateFlipPlateToTargetRotation(ResetRotation,FlipSpeed)
                    else: 
                        FinalLocal := (/Verse.org/SpatialMath:) transform{
                            Translation := FlipPlateEntity.GetLocalTransform().Translation,
                            Rotation := ResetRotation,
                            Scale := FlipPlateEntity.GetLocalTransform().Scale
                        }
                        FlipPlateEntity.SetLocalTransform(FinalLocal)
                    
                    # Enable Collider after everything is in place
                    if (ColliderComp :=  MaybeColliderPF?.GetColliderComp[]):
                        ColliderComp.Enable()
                    
                    UnflipCompleteEvent.Signal()


    RotateFlipPlateToTargetRotation(TargetRotation :(/Verse.org/SpatialMath:) rotation, RotSpeed : float, ?ArrivalThreashold: float = 0.01)<suspends> : void =
        if (FlipPlateEntity := MaybeFlipPlateEntity?):
            StartRotation := FlipPlateEntity.GetLocalTransform().Rotation
            loop:
                # Calculate the shortest rotation between current and target using AngularDistance
                PrevLocal := FlipPlateEntity.GetLocalTransform()
                RotationDelta := AngularDistance(FromRotation(PrevLocal.Rotation), FromRotation(TargetRotation))

                # Check if we're close enough to the target rotation
                if (RotationDelta < ArrivalThreashold) then break

                # Calculate the delta between frames
                Timestamp := GetSimulationElapsedTime()
                Sleep(0.0)
                DeltaTime := GetSimulationElapsedTime() - Timestamp

                # Calculate the rotation step based on speed and delta time
                Interp := Clamp(RotSpeed*DeltaTime, 0.0, 1.0)
                NewRotation := Slerp(PrevLocal.Rotation, TargetRotation, Interp)
                NewLocal :=  (/Verse.org/SpatialMath:)transform{
                    Translation := PrevLocal.Translation,
                    Rotation := NewRotation,
                    Scale := PrevLocal.Scale
                }
                
                # Apply new rotation
                FlipPlateEntity.SetLocalTransform(NewLocal)
            
            FinalLocal := (/Verse.org/SpatialMath:) transform{
                Translation := FlipPlateEntity.GetLocalTransform().Translation,
                Rotation := TargetRotation,
                Scale := FlipPlateEntity.GetLocalTransform().Scale
            }
            FlipPlateEntity.SetLocalTransform(FinalLocal)
    

    GetColliderComp()<decides><transacts>:void =
        ColliderPF := MaybeColliderPF?
        ColliderPF.GetColliderComp[]
    
    
    SetColliderDevice(NewObjectiveDevice: objective_device)<decides><transacts>:void =
        ColliderComp := MaybeColliderPF?.GetColliderComp[]
        set ColliderComp.MaybeObjectiveDevice = option. NewObjectiveDevice


(FlipSwitch: flip_switch).GetSwitchComp()<decides><transacts>:flip_switch_component =
    FlipSwitch.GetComponent[flip_switch_component]

(FlipSwitch: flip_switch).GetColliderComp()<decides><transacts>:damageable_collider_component =
    FlipSwitch.GetComponent[flip_switch_component].MaybeColliderPF?.GetColliderComp[]




    



    
        

    



