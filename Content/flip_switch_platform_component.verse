<#
@UUID: 9b27e5ac-12db-41b7-8bdc-4a9411782d43
@Task: Flip Switch Platform Component
@Description: A scene graph component for a circular, horiztonally rotating platform with 8 flip switches along the edge of the ring, and a single start flip switch at the center.
    Also uses a volume device to detect when a player is on the platform and can be used to prevent a player from leaving during the minigame.
@Tags: SceneGraph, Gameplay, Physics, Movement, PlayerInteraction
@Export: true
#>

<#
In-Editor Entity Prefab Setup:
    (NOTE: any setting not mentioned is assumed to be default or to your preference)


    
flip_switch
    RootEntity:
        - flip_switch_component
        DamageableCollider (damageable_collider entity prefab instance):
            - damageable_collider_component
                - MaybeObjectiveDevice: should be automatically set by the flip_switch_platform_component
            - transform_component
                - Translation: place slightly above FlipPlate
        FlipPlate:
            - mesh_component (preferrably a cylinder)
            - transform_component
                - Scale: tweak it until it is a flattened circle
  

platform_volume
    RootEntity:
        - platform_volume_component
            - MaybeVolume: set to an volume_device placed in the level


flip_switch_platform
    RootEntity:
        - flip_switch_platform_component
            - StartSwitchInput:
                - MaybeFlipSwitch: Set to the StartSwitch Entity 
                - MaybeColliderDevice: Set to a objective_device in the level with following settings
                    - Mesh: Cylinder, Sphere, or Cube
                    - Owning Team: None
                    - Health: 320000 (as high as you can possibly make it)
                    - Collision During Games: "Only When Visible"
                    - Collides With :"Weapons Only"
            - ObjectiveDeviceColliders: populate with 8 DIFFERENT objective_device_colliders in the level with the following settings
                - Mesh: Cylinder
                - Owning Team: None
                - Health: 320000 (as high as you can possibly make it)
                - Collision During Games: "Only When Visible"
                - Collides With :"Weapons Only"
        VolumeProxyEntity (platform_volume prefab)
            - MaybeVolume: volume_device in the level with the following details panel settings
                - Volume Shape: Cylinder
                - Volume Height: 8
                - Volume Radius: 16
                - Moved to cover entire switch platform
        StartSwitch (flip_switch prefab):
            - transform_component
                - Translation: place in the center of the platform
        FlipSwitchContainer:
            Switch1 (flip_switch prefab):
            Switch2 (flip_switch prefab):
            Switch3 (flip_switch prefab):
            Switch4 (flip_switch prefab):
            Switch5 (flip_switch prefab):
            Switch6 (flip_switch prefab):
            Switch7 (flip_switch prefab):
            Switch8 (flip_switch prefab):
            - NOTE: place these along the outer edge of the platform
        MainPlatform:
            mesh_component (preferrably a cylinder)
            transform_component:
                - Scale: tweak it until it is a LARGE flattened circle that encapsulates all of the flip switches

#>

using { /Verse.org }
using { /Verse.org/Concurrency }
using { /Verse.org/Simulation }
using { /Verse.org/Native }
using { /Verse.org/SceneGraph }
using { /Verse.org/SpatialMath }
using { /Fortnite.com/Game }
using { /Fortnite.com/Playspaces }
using { /Fortnite.com/Devices }
using { /UnrealEngine.com/Temporary/SpatialMath }
using { /UnrealEngine.com/Temporary/Diagnostics }
using { /Fortnite.com/Characters }
using {EPs}

# Struct used to input flip_switch prefab and objective device pairs
flip_switch_input := struct<concrete>():
    @editable
    MaybeFlipSwitch: ?flip_switch = false
    @editable
    MaybeColliderDevice: ?objective_device = false


flip_switch_platform_component<public>:= class<final_super>(component):

    # ============ EDITABLE FIELDS =======================================
    # The rotational speed in degrees per second that 
    @editable_number(float):
        MinValue := option. 0.05
    RotSpeed: float = 0.3

    # If true, platform will rotate clockwise; counterlockwise, if false.
    @editable
    InvertRotDirection: logic = false

    # The flip_switch at the center of the platform that 
    @editable
    StartSwitchInput: flip_switch_input = flip_switch_input{}

    # If true, this component will override the colliders for all flip_switch_components descended from this entity with the objective_devices found in ObjectiveDeviceColliders
    # Set to false if you want to manually set the references for each flip_switch prefab in the level editor
    @editable
    OverrideFlipSwitchColliders: logic = true

    # Objective Device collider proxies that will be set to each flip_switch_components' if OverrideFlipSwitchColliders is true.
    # Should equal the number of flip switches attached to the flip_switch_platform prefab
    @editable
    ObjectiveDeviceColliders: []objective_device = array{}
    
    # ============ EVENTS ==============================================
    # Signals when the start switch is successfully hit by an instigator; passes the game_action_instigator that is responsible
    StartSwitchHitEvent<public>: event(agent) = event(agent){}

    # Signals when any of the switches (other than the starting switch) is hit and there is a valid instigator behind it; passes the game_action_instigator that is responsible
    AnyHitSwitchEvent<public>: event(agent) = event(agent){}
    
    # Signals when this device's OnEndSimulation function is called; used primarily to cancel async logic after the game ends or the entity is removed from the scene.
    EndSimulationEvent<public>: event() = event(){}
    
    # ============ INTERNAL FIELDS =======================================
    # Flag that indicates whether the platform should be rotating. set to true to begin rotating, set to false to stop
    var IsRotating: logic = false

    var MaybeStartSwitchComp: ?flip_switch_component = false

    var MaybeVolumeComp: ?platform_volume_component = false

    # Contains list of all valid flip_switch_components found during initialization
    var SwitchComps: []flip_switch_component = array{}

    # Subscripiton to the PostPhysics Tick Event
    var MaybePostTickHandler: ?cancelable = false
    
    # Axis up vector for horizontal rotation
    RotationAxis : (/Verse.org/SpatialMath:)vector3 = (/Verse.org/SpatialMath:)vector3{Forward := 0.0, Left := 0.0, Up := 1.0}

    OnBeginSimulation<override>():void=
        (super:)OnBeginSimulation()

        if (StartSwitchComp := StartSwitchInput.MaybeFlipSwitch?.GetSwitchComp[]):
            set StartSwitchComp.StartGameFlipped = false    # must be done on OnBeginSimulation to ensure that it is unflipped on round start
            set StartSwitchComp.UnflipOnTimer = false
            set MaybeStartSwitchComp = option. StartSwitchComp
        else:
            Print("flip_switch_platform_component\nERROR: couldn't find start switch comp. Check if StartSwitchInput is set properly in the details panel")

        for:
            SwitchComp : Entity.FindDescendantComponents(flip_switch_component)
            (not MaybeStartSwitchComp?) or (SwitchComp <> MaybeStartSwitchComp?)  # if MaybeStartSwitchComp is set then make sure Switch comp isn't equal to it
        do:
            set SwitchComp.StartGameFlipped = true
            set SwitchComps += array{SwitchComp}

        set MaybeVolumeComp = option. Entity.FindDescendantComponentsArray(platform_volume_component)[0]

        # Set up On Round Started callback
        if (RoundMan := Entity.GetFortRoundManager[]):
            RoundMan.SubscribeRoundStarted(OnRoundStarted)


    OnEndSimulation<override>():void=
        (super:)OnEndSimulation()
        if. MaybePostTickHandler?.Cancel()
        EndSimulationEvent.Signal()


    OnRoundStarted()<suspends>:void =
        race:
            EndSimulationEvent.Await()
            block:
                # Subscribe to post-physics tick events
                set MaybePostTickHandler = option. TickEvents.PostPhysics.Subscribe(OnPostPhysics)

                # Assign a Collider to each flip_switch_component
                for:
                    Index -> SwitchComp : SwitchComps
                    Collider := ObjectiveDeviceColliders[Index]
                    SwitchComp.SetColliderDevice[Collider]
                do:
                    spawn. ListenForSwitchHitEvent(SwitchComp)
                    
                # Attempt to Initialize the Start Switch and listen for its hit event
                if:
                    StartSwitchComp := MaybeStartSwitchComp?
                    ColliderDevice := StartSwitchInput.MaybeColliderDevice?
                    StartSwitchComp.SetColliderDevice[ColliderDevice]
                then:
                    loop:
                        Result : damage_result = StartSwitchComp.HitEvent.Await()
                        if:
                            InstigatorFC := fort_character[Result.Instigator?]
                            InstigatorAgent := InstigatorFC.GetAgent[]
                        then:
                            StartSwitchHitEvent.Signal(InstigatorAgent)


    TurnOff( ?MaybeFlipDelay : ?float = false)<suspends>:void=
        set IsRotating = false
        FlipAllSwitches(?MaybeFlipDelay := MaybeFlipDelay)


    TurnOn(?MaybeFlipDelay: ?float = false)<suspends>:void=
        set IsRotating = true
        UnflipAllSwitches(?MaybeFlipDelay :=  MaybeFlipDelay)


    OnPostPhysics(Delta: float): void =
        if (IsRotating?):
            PrevLocal := Entity.GetLocalTransform()
            
            RotDirectionFactor := -1.0*InvertRotDirection.ToFloatSign()
            RotationDelta := MakeRotation(FromVector3(RotationAxis), RotDirectionFactor*RotSpeed*Delta)
            NewLocalRot := FromRotation(FromRotation(PrevLocal.Rotation).RotateBy(RotationDelta))

            NewLocal := (/Verse.org/SpatialMath:)transform{
                Translation := PrevLocal.Translation,
                Rotation := NewLocalRot,
                Scale := PrevLocal.Scale
            }
            # Apply new rotation
            Entity.SetLocalTransform(NewLocal)

    
    ListenForSwitchHitEvent(SwitchComp: flip_switch_component)<suspends>: void =
        race:
            EndSimulationEvent.Await()
            loop:
                Result : damage_result = SwitchComp.HitEvent.Await()
                if:
                    InstigatorFC := fort_character[Result.Instigator?]
                    InstigatorAgent := InstigatorFC.GetAgent[]
                then:
                    AnyHitSwitchEvent.Signal(InstigatorAgent)
                Sleep(0.0)

    
    FlipAllSwitches<public>(?MaybeFlipDelay:?float = false)<suspends>:void =
        for (SwitchComp : SwitchComps):
            spawn. SwitchComp.Flip()
            if (FlipDelay := MaybeFlipDelay?):
                Sleep(FlipDelay)


    UnflipAllSwitches<public>(?MaybeFlipDelay:?float = false)<suspends>:void =
        for (SwitchComp : SwitchComps):
            spawn. SwitchComp.Unflip()
            if (FlipDelay := MaybeFlipDelay?):
                Sleep(FlipDelay)


    FlipStartSwitch<public>()<suspends>:void =
        if (StartSwitchComp := MaybeStartSwitchComp?):
            StartSwitchComp.Flip()


    UnflipStartSwitch<public>()<suspends>:void =
        if (StartSwitchComp := MaybeStartSwitchComp?):
            StartSwitchComp.Unflip()

  
(SwitchPlatformPF: flip_switch_platform).GetSwitchPlatformComp()<decides><transacts>:flip_switch_platform_component = 
    SwitchPlatformPF.GetComponent[flip_switch_platform_component]

    