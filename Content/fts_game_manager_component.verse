<#
@UUID: 9b27e5ac-12db-41b7-8bdc-4a9411782d43
@Task: Flip The Switch Game Manager
@Description: Component that handles the top-level logic for the Flip the Switch Minigame. Manages player readying and registering, score tracking, most item granting.
@Tags: SceneGraph, Gameplay, DataStructures, PlayerManagement
@Export: true
#>

<#
In-Editor Entity Prefab Setup:
    (NOTE: any setting not mentioned is assumed to be default or to your preference)

flip_switch
    RootEntity:
        - flip_switch_component
        DamageableCollider (damageable_collider entity prefab instance):
            - damageable_collider_component
                - MaybeObjectiveDevice: should be automatically set by the flip_switch_platform_component
            - transform_component
                - Translation: place slightly above FlipPlate
        FlipPlate:
            - mesh_component (preferrably a cylinder)
            - transform_component
                - Scale: tweak it until it is a flattened circle


raised_platform
    RootEntity:
        - raised_platform_component
        PlatformPillar:
            - mesh_component (prefarably any with a flat top surface a cylinder or cube)
            - transform_component:
                - Tweak scale and translation such that it is a vertically tall pillar whose top is just barely sticking out of the ground
        VolumeProxyEntity (platform_volume prefab)
            - MaybeVolume: volume_device in the level with the following details panel settings
                - Volume Shape: Cylinder
                - Volume Height: 1
                - Volume Radius: 1
                - Moved to cover top of platform


flip_switch_platform
    RootEntity:
        - flip_switch_platform_component
            - StartSwitchInput:
                - MaybeFlipSwitch: Set to the StartSwitch Entity 
                - MaybeColliderDevice: Set to a objective_device in the level with following settings
                    - Mesh: Cylinder, Sphere, or Cube
                    - Owning Team: None
                    - Health: 320000 (as high as you can possibly make it)
                    - Collision During Games: "Only When Visible"
                    - Collides With :"Weapons Only"
            - ObjectiveDeviceColliders: populate with 8 DIFFERENT objective_device_colliders in the level with the following settings
                - Mesh: Cylinder
                - Owning Team: None
                - Health: 320000 (as high as you can possibly make it)
                - Collision During Games: "Only When Visible"
                - Collides With :"Weapons Only"
        VolumeProxyEntity (platform_volume prefab)
            - MaybeVolume: volume_device in the level with the following details panel settings
                - Volume Shape: Cylinder
                - Volume Height: 8
                - Volume Radius: 16
                - Moved to cover entire switch platform
        StartSwitch (flip_switch prefab):
            - transform_component
                - Translation: place in the center of the platform
        FlipSwitchContainer:
            Switch1 (flip_switch prefab):
            Switch2 (flip_switch prefab):
            Switch3 (flip_switch prefab):
            Switch4 (flip_switch prefab):
            Switch5 (flip_switch prefab):
            Switch6 (flip_switch prefab):
            Switch7 (flip_switch prefab):
            Switch8 (flip_switch prefab):
            - NOTE: place these along the outer edge of the platform
        MainPlatform:
            mesh_component (preferrably a cylinder)
            transform_component:
                - Scale: tweak it until it is a LARGE flattened circle that encapsulates all of the flip switches



fts_game_template
    RootEntity:
        - fts_hud_component
        - fts_game_manager_component
            - MaybeGunGranter: item_granter_device in the level with the following details pannel settings.
                - Item List: configure to grant 1 mechanical Bow (or any other ranged weapon of your choice)
                - Drop Items at Player Location: Never
            - MaybeGunRemover: item_granter_device in the level with following details panel settings
                - Configure to Remove the item granted by MaybeGunGranter
            - MaybeRaisedPlatform: RaisedPlatform root entity
            - MaybeSwitchPlatform: SwitchPlatform root entity
        SwitchPlatform (flip_switch_platform prefab):
        RaisedPlatform (raised_platform prefab):
    SwitchPlatform (flip_switch_platform prefab):
    RaisedPlatform (raised_platform prefab):

#>


using { /Verse.org }
using { /Verse.org/Concurrency }
using { /Verse.org/Simulation }
using { /Verse.org/Native }
using { /Verse.org/SceneGraph }
using { /Fortnite.com/Game }
using { /Fortnite.com/Teams }
using { /Fortnite.com/Characters }
using { /Fortnite.com/Playspaces }
using { /Fortnite.com/Devices }
using { /UnrealEngine.com/Temporary/SpatialMath }
using { /UnrealEngine.com/Temporary/Diagnostics }
using {EPs}


fts_side := enum {ShootingPlayer,MeleePlayer}

(Side: fts_side).Opponent()<computes>:fts_side =
    case (Side):
        fts_side.ShootingPlayer => fts_side.MeleePlayer
        fts_side.MeleePlayer => fts_side.ShootingPlayer

ToString(Side: fts_side)<transacts>:string =
    case (Side):
        fts_side.ShootingPlayer => "Shooting Player"
        fts_side.MeleePlayer =>     "Melee Player"

side_info := class():
    Side<public>: fts_side 
    var MaybeAgent<public>: ?agent = false
    var Score<public>: int = 0

(SideInfo: side_info).Reset():void =
    set SideInfo.MaybeAgent = false
    set SideInfo.Score = 0

score_update_result := struct:
    ScoringSide: fts_side
    NewScore: int

ready_update_result := struct:
    IsShootingSideReady<public>: logic
    IsMeleeSideReady<public>: logic
    
###################### Start Flip the Switch Game Manager Component #############################################################  

fts_game_manager_component<public> := class<final_super>(component):

    # Number of points a player's team recieves when they hit an opponent with a snowball
    @editable:
    PointsPerHit: int = 1

    # The number of points that either player must score to win the game
    @editable_number(int):
        MinValue := option. 1
    ScoreGoal: int = 15
    
    # The number of points that a player recieves when they hit a flip switch
    @editable_number(int):
        MinValue := option. 1
    PointPerHit: int = 1

    # Number of whole seconds that must pass after both sides are ready fo the game to begin
    @editable_number(int):
        MinValue := option. 0
    CountdownStartTime: int = 3
    
    # Amount of time, in seconds, that must pass after a game before everything resets
    @editable_number(float):
        MinValue := option. 0.0
    PostGameDelay: float = 4.0
    
    # Ranged weapon that is granted to the shooting player at the start of the minigame
    @editable
    MaybeGunGranter: ?item_granter_device = false
    
    # Corresponding weapon remover device used to remove item from the shooting player's inventory after the game 
    @editable
    MaybeGunRemover: ?item_remover_device = false

    # Isometric Camera for the Melee Player
    @editable
    MaybeMeleeCam: ?gameplay_camera_fixed_angle_device = false
    
    # Isometric Controls for the Melee Player
    @editable
    MaybeMeleeControls: ?gameplay_controls_third_person_device = false
    
    # Reference to the root entity of the raised_platform prefab 
    @editable
    MaybeRaisedPlatform: ?raised_platform = false

    # Reference to the root entity of the flip_switch_platform prefab 
    @editable
    MaybeSwitchPlatform: ?flip_switch_platform = false

    # ============ EVENTS =======================================
    # Signals when a player changes their participatiokn status in the game; passes the ready status of both players as logics
    ReadyStatusUpdateEvent: event(ready_update_result) = event(ready_update_result){}

    # Signals when a player's score updates
    ScoreUpdatedEvent: event(score_update_result) = event(score_update_result){}

    # Signals each time the pregame countdown ticks down; passes the new value
    CountdownTimerUpdatedEvent: event(int) = event(int){}

    # Signals when the pregame countdown is prematurely canceled, passes team that was responsible (if there was one)
    CountdownCanceledEvent: event(?agent) = event(?agent){}

    # Signals when a player reaches the minimum requirements to begin the game.
    ParticipantAbandonedEvent: event(side_info) = event(side_info){}

    # Signals after the pregame countdown timer expires and when the game starts
    GameBeginEvent: event() = event(){}

    # Signals when the game ends (regardless of the result); passes the winning team's info struct (passes false if there was no winner)
    GameOverEvent: event(?side_info) = event(?side_info){}

    # Signals when the Reset function is called and when the Post game delay is over.
    GameResetEvent: event() = event(){}

    # Signals when this component's OnEndSimulation function fires; used to cancel asyncronous logic when this component stops simulating
    EndSimulationEvent<internal>: event() = event(){}

    # ============ INTERNAL FIELDS =======================================

    var ShootingSide: side_info = side_info{Side := fts_side.ShootingPlayer}
    var MeleeSide: side_info = side_info{Side := fts_side.MeleePlayer}
    
    # The fts_hud_component which should be a sibling to this component on the root of the  fts_game_template prefab
    var MaybeHUDComponent: ?fts_hud_component = false 

    OnBeginSimulation<override>():void=
        (super:)OnBeginSimulation()
        set MaybeHUDComponent = option. Entity.GetComponent[fts_hud_component]
        if (RoundMan := Entity.GetFortRoundManager[]):
            RoundMan.SubscribeRoundStarted(OnRoundStart)
        

    OnEndSimulation<override>():void=
        (super:)OnEndSimulation()
        ResetGame()
        EndSimulationEvent.Signal()
    
    
    UpdateReadyStatuses():void =
        IsShootingSideReady := logic. ShootingSide.MaybeAgent?
        IsMeleeSideReady := logic. MeleeSide.MaybeAgent?
        ReadyStatusUpdateEvent.Signal(ready_update_result{ 
            IsShootingSideReady := IsShootingSideReady,
            IsMeleeSideReady := IsMeleeSideReady
        })


    # Handles top-level game logic
    OnRoundStart()<suspends>: void =
        race:
            EndSimulationEvent.Await()
            ListenForEitherSideAbandon()
            block:
                if:
                    RaisedPlatform := MaybeRaisedPlatform?
                    RaisedPlatformComp := RaisedPlatform.GetComponent[raised_platform_component]
                    SwitchPlatform := MaybeSwitchPlatform?
                    SwitchPlatformComp := SwitchPlatform.GetComponent[flip_switch_platform_component]
                then:
                    loop:
                        # Pre-Game Loop
                        loop:   
                            ResetGame()
                            # The Melee Player enters the switch platform and hits the start switch with their pickaxe.
                            InstigatorAgent := SwitchPlatformComp.StartSwitchHitEvent.Await()
                            set MeleeSide.MaybeAgent = option. InstigatorAgent
                            UpdateReadyStatuses()

                            # ...which causes the platform to raise
                            RaisedPlatformComp.Raise()
                            MaybeRaisedAgent: ?agent = option. RaisedPlatformComp.GetAgentOnTopOfPlatform[]

                            # If there is another play on the platform when its fully raised then begin the countdown
                            if:
                                RaisedAgent := RaisedPlatformComp.GetAgentOnTopOfPlatform[]
                                RaisedAgent <> InstigatorAgent
                                GunGranter := MaybeGunGranter?
                            then:
                                set ShootingSide.MaybeAgent = option. RaisedAgent
                                UpdateReadyStatuses()
                                GunGranter.GrantItem(RaisedAgent)
                                EnableMeleePerspective()
  
                                CountdownComplete := InitiateCountdown(CountdownStartTime)
                                if (CountdownComplete?):
                                    break

                            Sleep(2.0)
                            
                        # On Game Start
                        if:
                            ShootingPlayer := ShootingSide.MaybeAgent?
                            MeleePlayer := MeleeSide.MaybeAgent?
                            RaisedPlatVolumeComp := RaisedPlatformComp.MaybeVolumeComp?
                            RaisedPlatVolumeComp.RegisterAgent[ShootingPlayer]
                            SwitchPlatVolumeComp := SwitchPlatformComp.MaybeVolumeComp?
                            SwitchPlatVolumeComp.RegisterAgent[MeleePlayer]
                        then:

                            GameBeginEvent.Signal()
                            spawn. SwitchPlatformComp.TurnOn()

                            MaybeWinner :?side_info = race:  
                                ListenForScoreUpdate()
                                block:
                                    RemainingSide: side_info = ParticipantAbandonedEvent.Await()
                                    option. RemainingSide

                            # Post Game
                            GameOverEvent.Signal(MaybeWinner)
                            # Allow both players to leave thier platform during the post game
                            RaisedPlatVolumeComp.UnregisterAll()
                            SwitchPlatVolumeComp.UnregisterAll()
                            # Turn the Melee Player's normal camera and controls back
                            DisableMeleePerspective()
                            SwitchPlatformComp.TurnOff(?MaybeFlipDelay := option. 0.5)

                        Sleep(PostGameDelay)
                
    # Counts down one second at a time starting from the starting count 
    # Returns true if count reaches 0, returns false if either team is no longer "ready" or end simulation event signals before the count expires
    InitiateCountdown(StartingCount: int)<suspends>:logic =
        var  CountdownTimer: int = StartingCount   
        race:
            EndSimulationEvent.Await()
            loop:
                if (not BothPlayersReady[]) then break
                Sleep(0.0)
                ReadyStatusUpdateEvent.Await()
            loop:
                CountdownTimerUpdatedEvent.Signal(CountdownTimer)
                if (CountdownTimer <= 0) then return true 
                Sleep(1.0)
                set CountdownTimer -= 1
        false


    # Continuously Listens for when the agent assigned to either side gets removed from the playspace
    ListenForEitherSideAbandon()<suspends>:void =
        if (Playspace := Entity.GetPlayspaceForEntity[]):
            race:
                EndSimulationEvent.Await()
                loop: 
                    QuittingAgent := Playspace.ParticipantRemovedEvent().Await()
                    if (QuittingSide := WhichSide(QuittingAgent)?):
                        set QuittingSide.MaybeAgent = false
                        UpdateReadyStatuses()
                    Sleep(0.0)

    # Continuously listens for when a player hits a flip switch to reward them with a point; returns the side that reaches the score goal first
    ListenForScoreUpdate()<suspends>:?side_info =
        if:
            Playspace := Entity.GetPlayspaceForEntity[]
            SwitchPlatform := MaybeSwitchPlatform?
            SwitchPlatformComp := SwitchPlatform.GetSwitchPlatformComp[]
        then:
            loop:
                NewAgent := SwitchPlatformComp.AnyHitSwitchEvent.Await()
                MaybeScoringSide := WhichSide(NewAgent)
                
                if (ScoringSideInfo := MaybeScoringSide?):
                    set ScoringSideInfo.Score += PointsPerHit
                    ScoreUpdatedEvent.Signal(score_update_result{ScoringSide := ScoringSideInfo.Side, NewScore := ScoringSideInfo.Score})
                    if (ScoringSideInfo.Score >= ScoreGoal):
                        return option. ScoringSideInfo
        false
    

    ResetGame():void=
        GameResetEvent.Signal()
        if:
            RaisedPlatformComp := MaybeRaisedPlatform?.GetRaisedPlatformComponent[]
            VolumeComp := RaisedPlatformComp.MaybeVolumeComp?
        then:
            spawn. RaisedPlatformComp.Lower()
            VolumeComp.UnregisterAll()

        if:
            SwitchPlatformComp := MaybeSwitchPlatform?.GetSwitchPlatformComp[]
            VolumeComp := SwitchPlatformComp.MaybeVolumeComp?
        then:
            spawn. SwitchPlatformComp.UnflipStartSwitch()
            VolumeComp.UnregisterAll()

        DisableMeleePerspective()

        if (GunRemover := MaybeGunRemover?, ShootingAgent := ShootingSide.MaybeAgent?):
            GunRemover.Remove(ShootingAgent)

        ShootingSide.Reset()
        MeleeSide.Reset()
        UpdateReadyStatuses()


    EnableMeleePerspective():void=
        if:
            Agent := MeleeSide.MaybeAgent?
            MeleeControls := MaybeMeleeControls?
            MeleeCam := MaybeMeleeCam?
        then:
            MeleeControls.AddTo(Agent)
            MeleeCam.AddTo(Agent)


    DisableMeleePerspective():void =
        if:
            MeleeControls := MaybeMeleeControls?
            MeleeCam := MaybeMeleeCam?
        then:
            MeleeControls.RemoveFromAll()
            MeleeCam.RemoveFromAll()


    # Succeeds if both player slots are 
    BothPlayersReady()<decides><transacts>:void =
        ShootingSide.MaybeAgent? and MeleeSide.MaybeAgent?
        

    # Succeeds and returns a side_info data structure if the agent is registered one of the two sides; fails otherwise
    WhichSide(Agent : agent)<transacts>:?side_info =
        if (Agent = ShootingSide.MaybeAgent?): 
            return option. ShootingSide
        else if (Agent = MeleeSide.MaybeAgent?):
            return option. MeleeSide
        false
        

    GetInfo(Side: fts_side)<reads>:side_info =
        case (Side):
            fts_side.ShootingPlayer => ShootingSide
            fts_side.MeleePlayer => MeleeSide
