<#
@UUID: 9b27e5ac-12db-41b7-8bdc-4a9411782d43
@Task: Flip The Switch HUD Component
@Description: Handles the pure-Verse UI of the Flip The Switch Minigame. Reacts to game manager events to update the team's scores and display a timer and the game's result.
@Tags: UI, SceneGraph, DataStructures, Formatting
@Export: true
#>


<#
In-Editor Entity Prefab Setup:
    (NOTE: any setting not mentioned is assumed to be default or to your preference)


fts_game_template
    RootEntity:
        - fts_hud_component
        - fts_game_manager_component
            - MaybeGunGranter: item_granter_device in the level with the following details pannel settings.
                - Item List: configure to grant 1 mechanical Bow (or any other ranged weapon of your choice)
                - Drop Items at Player Location: Never
            - MaybeGunRemover: item_granter_device in the level with following details panel settings
                - Configure to Remove the item granted by MaybeGunGranter
            - MaybeRaisedPlatform: RaisedPlatform root entity
            - MaybeSwitchPlatform: SwitchPlatform root entity
        SwitchPlatform (flip_switch_platform prefab):
        RaisedPlatform (raised_platform prefab):
    SwitchPlatform (flip_switch_platform prefab):
    RaisedPlatform (raised_platform prefab):
#>

using { /UnrealEngine.com/Temporary/UI }
using { /UnrealEngine.com/Temporary/SpatialMath }
using { /Fortnite.com/UI }
using { /Fortnite.com/Devices }
using { /Fortnite.com/Teams }
using { /Fortnite.com/Playspaces }
using { /Verse.org }    
using { /Verse.org/Colors }
using { /Verse.org/Native }
using { /Verse.org/SceneGraph }
using { /Verse.org/Simulation }
using { /Fortnite.com/Game }


###################### Start UI Component ##########################################################################

fts_hud_component<public> := class<final_super>(component):

    var PlayerToUI:[player]fts_hud = map{} # Stores the UI object for each player
    var MaybeGameManager: ?fts_game_manager_component = false  # Reference to the sibling chess_game_manager_component
    
    OnBeginSimulation<override>():void =
        (super:)OnBeginSimulation()

        if (RoundMan := Entity.GetFortRoundManager[]):
            RoundMan.SubscribeRoundStarted(OnRoundStart)

        # Get Reference to Sibling Game Manager Component and setup gameplay callbacks
        set MaybeGameManager = option. Entity.GetComponent[fts_game_manager_component] 

        if (Playspace := Entity.GetPlayspaceForEntity[]):
            Playspace.PlayerRemovedEvent().Subscribe(OnPlayerRemoved)
        

    OnRoundStart()<suspends>:void=
        Sleep(1.0)
        if (GameManager := MaybeGameManager?):
            AddUIAllPlayers()
            loop:
                # Waiting For Player Ready
                race:
                    GameManager.GameBeginEvent.Await()
                    loop: # Countdown
                        Result := GameManager.CountdownTimerUpdatedEvent.Await()
                        UpdateResultText(GameStartMessage("{Result}"))
                        Sleep(0.0)
                    loop: # On Player Ready
                        UpdateReadyStatus(GameManager.ReadyStatusUpdateEvent.Await())
                        Sleep(0.0)
                    
                UpdateResultText(Msg<localizes> : message = "")

                # Game Loop
                MaybeWinner :?side_info = race:
                    GameManager.GameOverEvent.Await()
                    block:  # On Score Updated
                        loop:
                            Result := GameManager.ScoreUpdatedEvent.Await()
                            case (Result.ScoringSide):
                                fts_side.ShootingPlayer => UpdatePlayer1PointsText(Result.NewScore)
                                fts_side.MeleePlayer =>
                                    Print("On Melee Player Score")
                                    UpdatePlayer2PointsText(Result.NewScore)
                            Sleep(0.0)
                        Sleep(Inf)
                        false


                # Post Game
                ResultMessage := if (Winner := MaybeWinner?):
                    UpdateResultText(PlayerWinMessage("The {ToString(Winner.Side)}"))
                else:
                    UpdateResultText(PlayerWinMessage("No One"))     
                
                GameManager.GameResetEvent.Await()
                ResetUI()
                
                    
    OnEndSimulation<override>():void =
        RemoveUIAllPlayers()    # Remove all UI elements on end; Custom Verse UI does not get cleaned up automatically.

    
    UpdateReadyStatus (UpdateResult: ready_update_result) : void = 
        if (not UpdateResult.IsShootingSideReady? and not UpdateResult.IsMeleeSideReady?):
            UpdateResultText(WaitingOnPlayers())

        else if (UpdateResult.IsShootingSideReady? and not UpdateResult.IsMeleeSideReady?):
            UpdateResultText(WaitingOnPlayer1())

        else if (UpdateResult.IsShootingSideReady? and not UpdateResult.IsMeleeSideReady?):
            UpdateResultText(WaitingOnPlayer2())

    
    UpdatePlayer1PointsText(NewVal:int):void =
        if (GameManager := MaybeGameManager?):
            MaxScore := GameManager.ScoreGoal
            for (Player->Hud : PlayerToUI):
                for (Index -> PointWidget : Hud.Player1PointWidgets):
                    if (Index >= MaxScore - NewVal):
                        PointWidget.Widget.SetVisibility(widget_visibility.Visible)
                    else:
                        PointWidget.Widget.SetVisibility(widget_visibility.Hidden)
        
    
    UpdatePlayer2PointsText(NewVal:int):void = 
        if (GameManager := MaybeGameManager?):
            for (Player->Hud : PlayerToUI):
                for (Index -> PointWidget : Hud.Player2PointWidgets):
                    if (Index >= NewVal):
                        PointWidget.Widget.SetVisibility(widget_visibility.Hidden)
                    else:
                        PointWidget.Widget.SetVisibility(widget_visibility.Visible)

    
    UpdateResultText(Message:message):void = 
        for (Player->Hud : PlayerToUI):
            Hud.ResultText.SetText(Message)

            
    ResetUI():void = 
        for (Player->Hud : PlayerToUI):
            Hud.Reset()


    AddUIAllPlayers():void =
        if(Playspace := Entity.GetPlayspaceForEntity[]):
            for (Player : Playspace.GetPlayers()) do Player.AddUI()


    RemoveUIAllPlayers():void=
        if(Playspace := Entity.GetPlayspaceForEntity[]):
            for(Player : Playspace.GetPlayers()):
                Player.RemoveUI()
            set PlayerToUI = map{}
            

    (Player : player).AddUI():void =
        if:
            PlayerUI := GetPlayerUI[Player]
            BattingGameUI := fts_hud{}
            set PlayerToUI[Player] = BattingGameUI
            GameManager := MaybeGameManager?
        then:
            BattingGameUI.CreateUI(GameManager.ScoreGoal)
            PlayerUI.AddWidget(BattingGameUI.GetWidget())
        else:
            Print("chess_hud_component:\nERROR: Could not add UI to player's screen")
        

    (Player : player).RemoveUI():void =
        if:
            PlayerUI := GetPlayerUI[Player]
            BattingGameUI := PlayerToUI[Player]
        then:
            PlayerUI.RemoveWidget(BattingGameUI.GetWidget())
            set PlayerToUI = PlayerToUI.RemoveKey(Player)
    

    OnPlayerAdded(Player: player):void = Player.AddUI()
    OnPlayerRemoved(Player: player):void = Player.RemoveUI()

###################### End UI Component #############################################################################


###################### Start UI Display Messages ######################################################################

TitleMessage<localizes>() : message = "Flip the Switch"
WaitingOnPlayer1<localizes>() : message = "Waiting for Player 1 to ready up"
WaitingOnPlayer2<localizes>() : message = "Waiting for Player 2 to ready up"
WaitingOnPlayers<localizes>() : message = "Waiting for both players to be ready"
GameStartMessage<localizes>(String: string) : message = "Game Starting in {String}"
PlayerWinMessage<localizes>(String: string) : message = "{String} is the victor!"

###################### End UI Display Messages ##########################################################################


###################### Start Player UI Class ############################################################################
                
fts_hud := class:

    ResultText : text_block = text_block{DefaultTextColor := NamedColors.Yellow}

    var MainWidget : canvas = canvas{}

    var Player1PointWidgets : []overlay_slot = array{}
    var Player1StackboxSlots : []stack_box_slot = array{}
    var Player2PointWidgets : []overlay_slot = array{}
    var Player2StackboxSlots : []stack_box_slot = array{}

    GetWidget() : canvas = MainWidget


    Reset() : void = 
        ResultText.SetText(WaitingOnPlayers())

        for (Index -> PointWidget : Player1PointWidgets):
            PointWidget.Widget.SetVisibility(widget_visibility.Hidden)

        for (Index -> PointWidget : Player2PointWidgets):
            PointWidget.Widget.SetVisibility(widget_visibility.Hidden)
    

    CreateNewPointWidget(Color : color):overlay_slot =
        overlay_slot{
            Widget := color_block:
                DefaultColor := Color
                DefaultOpacity := 0.65
                DefaultDesiredSize := vector2{X := 16.0, Y := 64.0}
            }


    CreateNewStackboxSlotPointWidget(PointOverlaySlot : overlay_slot):stack_box_slot =
        stack_box_slot{
            Padding := margin{Left := 0.0, Top := 0.0, Right := 0.0, Bottom := 0.0}
            Widget := overlay{
                
                Slots := array{
                    overlay_slot{
                        Widget := color_block:
                            DefaultColor := NamedColors.Black
                            DefaultOpacity := 0.65
                            DefaultDesiredSize := vector2{X := 24.0, Y := 72.0}
                    },

                    PointOverlaySlot
                }
            }
        }


    CreateUI(MaxScore : int) : void =
        GameTitle: text_block = text_block{DefaultTextColor := NamedColors.White, DefaultText := TitleMessage()},

        for(Index := 0..MaxScore-1):
            set Player1PointWidgets += array{CreateNewPointWidget(NamedColors.CornflowerBlue)}
            
            if(PointOverlaySlot := Player1PointWidgets[Index]):
                set Player1StackboxSlots += array{CreateNewStackboxSlotPointWidget(PointOverlaySlot)}

            set Player2PointWidgets += array{CreateNewPointWidget(NamedColors.ForestGreen)}

            if(PointOverlaySlot := Player2PointWidgets[Index]):
                set Player2StackboxSlots += array{CreateNewStackboxSlotPointWidget(PointOverlaySlot)}


        ScoreUI := canvas:
            Slots := array:
                # Create title at the top of the screen
                canvas_slot:
                    Anchors := anchors{Minimum := vector2{X:= 0.5, Y := 0.0}, Maximum := vector2{X:= 0.5, Y := 0.0}}
                    Offsets := margin{Left := 0.0, Top := 130.0, Right := 650.0, Bottom := 100.0}
                    Alignment := vector2{X := 0.5, Y := 0.5}
                    Widget := overlay{
                        Slots := array{
                            overlay_slot:
                                Widget := color_block:
                                    DefaultColor := MakeColorFromSRGB(1.0, 0.07, 0.07)
                                    DefaultOpacity := 0.5
                                    DefaultDesiredSize := vector2{X := 400.0, Y := 80.0}  

                            overlay_slot{
                                Widget := GameTitle
                            }
                        }
                    }

                # Create point boxes for player 1 and 2
                canvas_slot:
                    Anchors := anchors{Minimum := vector2{X:= 0.5, Y := 0.0}, Maximum := vector2{X:= 0.5, Y := 0.0}}
                    Offsets := margin{Left := 0.0, Top := 220.0, Right := 1000.0, Bottom := 100.0}
                    Alignment := vector2{X := 0.5, Y := 0.5}
                    Widget := stack_box{
                        Orientation := orientation.Horizontal,
                        Slots := array{
                            stack_box_slot{
                                HorizontalAlignment := horizontal_alignment.Left
                                Widget := stack_box{
                                Orientation := orientation.Horizontal,
                                Slots := Player1StackboxSlots
                                }
                            },

                            stack_box_slot{
                                HorizontalAlignment := horizontal_alignment.Right
                                Widget := stack_box{
                                Orientation := orientation.Horizontal,
                                Slots := Player2StackboxSlots
                                }
                            }
                        }
                    }
        
        #Adds a timer and winner/draw text underneath the score text above
        OverallUI := stack_box{
            Orientation := orientation.Vertical,
            Slots := array{
                stack_box_slot{
                    Widget := ScoreUI,
                    HorizontalAlignment := horizontal_alignment.Center,
                    VerticalAlignment := vertical_alignment.Center
                },
                stack_box_slot{
                    Widget := ResultText,
                    HorizontalAlignment := horizontal_alignment.Center,
                    VerticalAlignment := vertical_alignment.Center
                }
            }
        }

        set MainWidget = canvas{
            Slots := array{
                canvas_slot{
                    Widget := OverallUI,
                    Anchors := anchors{Minimum := vector2{X := 0.5, Y := 0.1}, Maximum := vector2{X := 0.5, Y := 0.1}},
                    Alignment := vector2{X := 0.5, Y := 0.5}
                }
            }
        }

        Reset()