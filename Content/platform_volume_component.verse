<#
@UUID: 9b27e5ac-12db-41b7-8bdc-4a9411782d43
@Task: Platform Volume Component
@Description: A component that uses a volume_device in the level to trap agents on a scene graph platform during gameplay. Also moves volume alongside entity
@Tags: SceneGraph, Gameplay, Movement, PlayerInteraction
@Export: true
#>

<#
In-Editor Entity Prefab Setup:
    (NOTE: any setting not mentioned is assumed to be default or to your preference)

platform_volume
    RootEntity:
        - platform_volume_component
            - MaybeVolume: set to an volume_device placed in the level

#>


using { /Verse.org }
using { /Verse.org/Concurrency }
using { /Verse.org/Simulation }
using { /Verse.org/Native }
using { /Verse.org/SceneGraph }
using { /Verse.org/SpatialMath }
using { /Fortnite.com/Game }
using { /Fortnite.com/Characters }
using { /Fortnite.com/Playspaces }
using { /Fortnite.com/Devices }
using { /UnrealEngine.com/Temporary/SpatialMath }
using { /UnrealEngine.com/Temporary/Diagnostics }
using {EPs}


platform_volume_component<public>:= class<final_super>(component):

    @editable
    var MaybeVolume: ?volume_device = false
    @editable
    var UseDeviceScale: logic = true

    var MaybePostTickHandler: ?cancelable = false
    var MaybeAgentExitHandler: ?cancelable = false
    var WillPreventEscapes: logic = true
    var TrappedAgents: [agent]logic = map{}

    OnBeginSimulation<override>():void=
        (super:)OnBeginSimulation()
        if (RoundMan := Entity.GetFortRoundManager[]):
            RoundMan.SubscribeRoundStarted(OnRoundStarted)


    OnRoundStarted()<suspends>:void =
        # Subscribe to post-physics tick events
        set MaybePostTickHandler = option. TickEvents.PostPhysics.Subscribe(OnPostPhysics)

        if(Volume := MaybeVolume?):
            Sub := Volume.AgentExitsEvent.Subscribe(OnAgentAttemptExit)
            set MaybeAgentExitHandler = option. Sub
            

    OnEndSimulation<override>():void=
        (super:)OnEndSimulation()
        if. MaybePostTickHandler?.Cancel()
        if. MaybeAgentExitHandler?.Cancel()
        UnregisterAll()


    OnPostPhysics(Delta: float):void = 
        if (IsSimulating[],Volume := MaybeVolume?):
            DeviceTransform := Volume.GetTransform()
            GlobalTransform := Entity.GetGlobalTransform()
            NewScale : (/UnrealEngine.com/Temporary/SpatialMath:) vector3 = if (UseDeviceScale?) then DeviceTransform.Scale else FromVector3(GlobalTransform.Scale)

            FinalGlobal :=  (/UnrealEngine.com/Temporary/SpatialMath:) transform{
                Translation := FromVector3(GlobalTransform.Translation),
                Rotation := FromRotation(GlobalTransform.Rotation),
                Scale := NewScale
            }
            if. Volume.TeleportTo[FinalGlobal]
    

    OnAgentAttemptExit(Agent: agent):void =
        if:
            WillPreventEscapes?
            IsSimulating[]
            TrappedAgents[Agent]
            Volume := MaybeVolume?        
            AgentFC := Agent.GetFortCharacter[]    
            AgentFC.TeleportTo[FromVector3(Entity.GetGlobalTransform().Translation), AgentFC.GetTransform().Rotation]
        then:
            Print("Teleport successful")
    
    
    RegisterAgent(Agent: agent)<decides><transacts>:void =
        not TrappedAgents[Agent]
        set TrappedAgents[Agent] = true
        
    UnregisterAgent(Agent: agent)<decides><transacts>:void =
        TrappedAgents[Agent]
        set TrappedAgents = TrappedAgents.RemoveKey(Agent)

    UnregisterAll():void =
        set TrappedAgents = map{}

    



