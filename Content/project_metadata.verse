<#
@UUID: 9b27e5ac-12db-41b7-8bdc-4a9411782d43
@Task: Flip The Switch Minigame
@Description: Use scene graph to build an arcade-style minigame where 2 players asymetrically compete against one another by hitting switches that temporarily 
    flip when damaged. One player is the "Melee Player", who plays with a top-down perspective of the large spinning platform and uses their pickaxe to hit switches.
    The second player is the "Shooting Player," who is placed on a raised platform and given a bow to hit the same flip switches from a distance. Before the 
    game begins, the player who wishes to be the "Shooting Player" must stand on an unraised pedestool while the Melee Player hits a Starting switch in the
    center of the large platform. Doing so causes the unraised pedestool to rise. A countdown begins if a player is present on the pedestool when it stops rising.
    Once the countdown completes, the Shooting Player is granted a mechanical bow with infinite ammo, the large platform begins spinning, all the flip switches unflip,
    and the game officially begins. Each time either player hits a flip switch, they are granted a point  and that switch goes on a cooldown before it can be hit again.
    Each player's current score is displayed on a HUD. If a player tries to leave their respective play area  (Melee = large platform; Shooting Player = raised platform)
    before the game is over, they are automatically teleported back. When either player reaches a configurable score number of points, they are declared the winner and the
    game is over.
@Tags: PvP, NonCombat, Arcade, Shooter, Top-Down, Asymetric
@Export: true
#>