<#
@UUID: 9b27e5ac-12db-41b7-8bdc-4a9411782d43
@Task: Flip Raised Platform Component
@Description: Scene graph component that controls a rising platform the Shooting Player stands on during the game. 
@Tags: SceneGraph, Gameplay, Physics, Movement, PlayerInteraction
@Export: true
#>

<#
In-Editor Entity Prefab Setup:
    (NOTE: any setting not mentioned is assumed to be default or to your preference)

platform_volume
    RootEntity:
        - platform_volume_component
            - MaybeVolume: set to an volume_device placed in the level


raised_platform
    RootEntity:
        - raised_platform_component
        PlatformPillar:
            - mesh_component (prefarably any with a flat top surface a cylinder or cube)
            - transform_component:
                - Tweak scale and translation such that it is a vertically tall pillar whose top is just barely sticking out of the ground
        VolumeProxyEntity (platform_volume prefab)
            - MaybeVolume: volume_device in the level with the following details panel settings
                - Volume Shape: Cylinder
                - Volume Height: 1
                - Volume Radius: 1
                - Moved to cover top of platform

#>


using { /Verse.org }
using { /Verse.org/Concurrency }
using { /Verse.org/Simulation }
using { /Verse.org/Native }
using { /Verse.org/SceneGraph }
using { /Verse.org/SpatialMath }
using { /Fortnite.com/Game }
using { /Fortnite.com/Playspaces }
using { /Fortnite.com/Devices }
using { /UnrealEngine.com/Temporary/SpatialMath }
using { /UnrealEngine.com/Temporary/Diagnostics }
using { /Verse.org/Random }
using { /Fortnite.com/Characters }

raised_platform_component<public>:= class<final_super>(component):

    # ============ EDITABLE FIELDS =======================================
    # The vertical offset of the platform 
    @editable_slider(float):
        MinValue := option. 0.0
        MaxValue := option. 1000.0
    MaxHeightOffset: float = 600.0

    # The speed the platform moves
    @editable_number(float):
        MinValue := option. 0.1
    MoveSpeed: float = 1.0
    
    # ============ EVENTS ==============================================
    # Signals when the MoveToHeightFunction is first called; passes the new height this platform is moving towards
    TargetHeightOffsetChangedEvent<public>: event(float) =  event(float){}

    # Signals when the platform has completed a movement; passes the platform's new height
    CompletedMovementEvent<public>: event(float) = event(float){}

    # Signals when this device's OnEndSimulation function is called; used primarily to cancel async logic after the game ends or the entity is removed from the scene.
    EndSimulationEvent<public>: event() = event(){}
    
    # ============ INTERNAL FIELDS =======================================
    # var MaybeEntity: ?entity = false
    var MaybeVolumeComp: ?platform_volume_component = false
    var OriginGlobalHeight: float = 0.0
    var TargetHeightOffset<internal>: float = 0.0
    var MaybePostTickHandler: ?cancelable = false


    OnBeginSimulation<override>():void=
        (super:)OnBeginSimulation()

        set OriginGlobalHeight = Entity.GetGlobalTransform().Translation.Up

        if (VolumeComp := Entity.FindDescendantComponentsArray(platform_volume_component)[0]):
            set MaybeVolumeComp = option. VolumeComp
        
        set TargetHeightOffset = OriginGlobalHeight
    
    
    OnEndSimulation<override>():void=
        (super:)OnEndSimulation()
        EndSimulationEvent.Signal()


    Raise<public>()<suspends>: void =
        if (IsFullyRaised[] or TargetHeightOffset = MaxHeightOffset) then return 
        MoveToNewHeightOffset(MaxHeightOffset)


    Lower<public>()<suspends>:void =
        if (IsFullyLowered[] or TargetHeightOffset = 0.0) then return
        MoveToNewHeightOffset(0.0)


    MoveToNewHeightOffset(NewHeightOffset : float)<suspends>:void=
        if (NewHeightOffset <> TargetHeightOffset):
            # Cancel any ongoing movements 
            TargetHeightOffsetChangedEvent.Signal(NewHeightOffset)
            # Make sure that we can't go past the min or max height
            set TargetHeightOffset = Clamp(NewHeightOffset,0.0,MaxHeightOffset)
            StartPos := Entity.GetGlobalTransform().Translation
            EndPos := (/Verse.org/SpatialMath:) vector3{Left:=StartPos.Left, Forward:=StartPos.Forward, Up:=OriginGlobalHeight+TargetHeightOffset}
            race:
                TargetHeightOffsetChangedEvent.Await()
                EndSimulationEvent.Await()  
                block:
                    Timestamp := GetSimulationElapsedTime()
                    loop:
                        Sleep(0.0)
                        DeltaTime := GetSimulationElapsedTime() - Timestamp
                        Interp := Clamp(MoveSpeed*DeltaTime, 0.0, 1.0)
                        CurrentGlobal := Entity.GetGlobalTransform()
                        NewTransform := (/Verse.org/SpatialMath:)transform{
                            Translation := Lerp(StartPos, EndPos, Interp),
                            Rotation := CurrentGlobal.Rotation,
                            Scale := CurrentGlobal.Scale
                        }
                        Entity.SetGlobalTransform(NewTransform)
                        if (Interp >= 1.0):
                            CompletedMovementEvent.Signal(NewHeightOffset)
                            break

            # After Move animation complete, make sure its in the right spot
            CurrentTransform := Entity.GetGlobalTransform()
            FinalTransform := (/Verse.org/SpatialMath:)transform{
                Translation := EndPos,
                Rotation := CurrentTransform.Rotation,
                Scale := CurrentTransform.Scale
            }
            Entity.SetGlobalTransform(FinalTransform)


    IsFullyRaised()<decides><transacts>: void =
        Entity.GetGlobalTransform().Translation.Up >= OriginGlobalHeight + MaxHeightOffset
        

    IsFullyLowered()<decides><transacts>: void = 
        Entity.GetGlobalTransform().Translation.Up <= OriginGlobalHeight
    
    
    GetAgentOnTopOfPlatform()<decides><transacts>:agent =
        Volume := MaybeVolumeComp?.MaybeVolume?
        Agents := Volume.GetAgentsInVolume()
        Agents[0]
        


(RaisedPlatformPF: EPs.raised_platform).GetRaisedPlatformComponent()<decides><transacts>:raised_platform_component = 
    RaisedPlatformPF.GetComponent[raised_platform_component]

        
                