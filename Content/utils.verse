<#
@UUID: 9b27e5ac-12db-41b7-8bdc-4a9411782d43
@Task: Flip The Switch Minigame Utils
@Description: All the miscellaneous helper functions used for this project
@Tags: Utility, DataConversion, Formatting
@Export: true
#>

using { /Verse.org/SceneGraph }
using { /Verse.org/Concurrency }
using { /Verse.org/Simulation }
using { /Fortnite.com/Playspaces }
using { /Fortnite.com/Teams }
using { /Verse.org/Colors }
using { /UnrealEngine.com/Temporary/Diagnostics }





(Val:logic).ToIntSign():int = if. Val? then 1 else -1
(Val:logic).ToFloatSign():float = if. Val? then 1.0 else -1.0

# Conatenate 2 messages
operator'+'<localizes><public>(M1:message, M2:message)<transacts>:message="{M1}{M2}"

EmptyMessage<localizes><public>() : message = ""

FloatToMessage<localizes><public>(Float:float):message = "{Float}"
IntToMessage<localizes><public>(Num:int):message = "{Num}"
StringToMessage<localizes>(InString:string):message = "{InString}"


# Helper functions used to manually throw a failure in a more readable way
Fail<public>()<decides><computes>:void= false?
# Fail overload that allows it to be used in function that returns a type
Fail<public>(t:type)<decides><transacts>:t=
    Failer:?t = false
    Failer?    


# Removes an element from the given map and returns a new map without that element
(Map:[keytype]valuetype where keytype:subtype(comparable), valuetype:type).RemoveKey(ElementToRemove:keytype)<transacts>:[keytype]valuetype=
    var NewMap:[keytype]valuetype = map{}
    for (Key -> Value : Map, Key <> ElementToRemove):
        set NewMap = ConcatenateMaps(NewMap, map{Key => Value})
    return NewMap

# Helper function to return an array (instead of a generator) of all components attached to descendant entities to and including `InEntity` of type `component_type`.
(Entity:entity).FindDescendantComponentsArray<public>(ComponentType:castable_subtype(component))<transacts>:[]ComponentType =
    var Components:[]ComponentType = array{}
    for (Comp : Entity.FindDescendantComponents(ComponentType)):
        set Components += array{Comp}
    Components

# Helper function to return an array (instead of a generator) of all components attached to descendant entities to and including `InEntity` of type `component_type`.
(Entity:entity).FindAncestorComponentsArray<public>(ComponentType:castable_subtype(component))<transacts>:[]ComponentType =
    var Components:[]ComponentType = array{}
    for (Comp : Entity.FindAncestorComponents(ComponentType)):
        set Components += array{Comp}
    Components


